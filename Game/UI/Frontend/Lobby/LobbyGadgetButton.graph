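# LobbyGadgetButton is a lobby widget for one gadget slot, with an item card inside a panel button over an empty-slot image.
# Refresh stores a world item and feeds the card, hiding it when the item is invalid.
# SetLocked disables the button, while click and hover are re-broadcast through dispatchers passing the widget itself.
# At design time, PreConstruct previews a random test gadget through a temporary item instance.
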WidgetBlueprint /Game/UI/Frontend/Lobby/LobbyGadgetButton
graphs [
 EventGraph {nodes [
   "On Button Clicked (Button)"!() -> (OutputDelegate, then, Button:Common Button Legacy Object Reference) {is=title K2Node_ComponentBoundEvent}
   > Call On Button Clicked!(execute:Exec, self:Lobby Gadget Button Object Reference, GadgetButton:Lobby Gadget Button Object Reference=Self-Reference()) -> (then:Exec) {is=title K2Node_CallDelegate}
   "On Button Hovered (Button)"!() -> (OutputDelegate, then, Button:Common Button Legacy Object Reference) {is=title K2Node_ComponentBoundEvent}
   > Call On Button Hovered!(execute:Exec, self:Lobby Gadget Button Object Reference, GadgetButton:Lobby Gadget Button Object Reference=Self-Reference()) -> (then:Exec) {is=title K2Node_CallDelegate}
   n1 = Event PreConstruct()
   > Branch(Condition=@n1.IsDesignTime)
   n2 = RandomInteger!(Max:int=Length(TargetArray:Array of Fort World Item Definition Object References=Get DesignTestItemDefinitions() -> (DesignTestItemDefinitions:Array of Fort World Item Definition Object References))) -> (ReturnValue:int=0)
   n3 = Get DesignTestItemDefinitions()
   > SetItemToRepresent!(
    execute
    self:Fort Item Widget NUI Object Reference=Get Item() -> (Item:Fort Multi Size Item Card Object Reference)
    ItemToRepresent:Fort Item Object Reference=CreateTemporaryItemInstance!(
     self:Fort Item Definition Object Reference="Get (a copy)"(Array:Array of Fort World Item Definition Object References=@n3, Dimension 1=@n2) -> (Output:Fort World Item Definition Object Reference)
     Count:int=1
     Level:int=-1
    ) -> (ReturnValue:Fort Item Object Reference)
   ) -> (then)
  ]}
 GetGadgetItem {nodes [
   GetGadgetItem() {is=title K2Node_FunctionEntry}
   > Return Node!(execute, GadgetItem:Fort World Item Object Reference=Get GadgetItem() -> (GadgetItem:Fort World Item Object Reference))
  ]}
 GetLocked {nodes [
   GetLocked() {is=title K2Node_FunctionEntry}
   > Return Node!(execute, bLocked:bool=Get bLocked() -> (bLocked:bool))
  ]}
 OnButtonClicked {nodes [OnButtonClicked!() -> (then, GadgetButton:Lobby Gadget Button Object Reference) {is=title K2Node_FunctionEntry}]}
 OnButtonHovered {nodes [OnButtonHovered!() -> (then, GadgetButton:Lobby Gadget Button Object Reference) {is=title K2Node_FunctionEntry}]}
 Refresh {nodes [
   n1 = Refresh!() -> (then, WorldItem:Fort World Item Object Reference) {is=title K2Node_FunctionEntry}
   > Set GadgetItem(GadgetItem:Fort World Item Object Reference=@n1) -> (Output_Get:Fort World Item Object Reference)
   > SetItemToRepresent(self=Get Item(), ItemToRepresent=Get GadgetItem())
   > n2 = Branch(Condition=IsValid(Object=Get GadgetItem()))
   > SetVisibility(self=Get Item(), InVisibility=SelfHitTestInvisible)
   SetVisibility(execute=@n2.else, self=Get Item(), InVisibility=Hidden)
  ]}
 SetLocked {nodes [
   n1 = SetLocked!() -> (then, bLocked:bool=~) {is=title K2Node_FunctionEntry}
   > n2 = Set bLocked(bLocked:bool=@n1) -> (Output_Get:bool)
   > SetIsEnabled(self=Get Button() -> (Button:Panel Button Object Reference), bInIsEnabled=NOT Boolean(A=@n2))
  ]}
]
variables [
 bLocked {type=boolean}
 DesignTestItemDefinitions {
  type="array<FortWorldItemDefinition>"
  default [/Game/Abilities/Player/Generic/Gadgets/AdrenalineRush/G_Generic_AdrenalineRush, /Game/Abilities/Player/Generic/Gadgets/AirStrike/G_AirStrike]
 }
 GadgetItem {type=FortWorldItem, default=null}
 OnButtonClicked {}
 OnButtonHovered {}
]
widgets {
 info {inheritedWidgetCount=0, namedSlotCount=1, parentClass=&CommonUserWidget, rootWidgetClass=&Overlay, widgetCount=4}
 widgets [
  Overlay_0 {widgetClassPath=&Overlay, widget=Overlay_0}
  EmptySlot {
   widgetClassPath=&Image
   bIsVariable=true
   parent=Overlay_0
   widget=EmptySlot
   properties {brush {
     bIsDynamicallyLoaded=false
     drawAs=Image
     imageType=NoImage
     mirroring=NoMirror
     resourceName=None
     resourceObject=/FortUILibrary/Foundation/Buttons/Materials/Slot/M_Slot_Gadget_S_Empty
     tiling=NoTile
     imageSize=<112 112>
     margin=<ltrb 0 0 0 0>
     outlineSettings {
      bUseBrushTransparency=false
      roundingType=HalfHeightRadius
      width=0
      color {colorUseRule=UseColor_Specified, specifiedColor=<rgba 0 0 0 0>}
      cornerRadii=<0 0 0 0>
     }
     tintColor {colorUseRule=UseColor_Specified, specifiedColor=<rgba 1 1 1 1>}
     uVRegion {bIsValid=false, max=<0 0>, min=<0 0>}
    }}
   slot {class=OverlaySlot, properties {horizontalAlignment=HAlign_Fill, verticalAlignment=VAlign_Fill}}
  }
  Button {
   widgetClassPath=/FortUILibrary/Foundation/Buttons/PanelButton
   bIsVariable=true
   parent=Overlay_0
   widget=Button
   properties {
    onButtonBaseClicked=(&self:WidgetTree.Button.HandleOnButtonClicked)
    onButtonBaseDoubleClicked=(&self:WidgetTree.Button.HandleOnButtonDoubleClicked)
    onButtonBaseHovered=(&self:WidgetTree.Button.HandleOnButtonHovered)
    onButtonBaseUnhovered=(&self:WidgetTree.Button.HandleOnButtonUnhovered)
    onSelectedChangedBase=(&self:WidgetTree.Button.HandleOnSelectedChanged)
    style=/FortUILibrary/Foundation/Buttons/ButtonStyles/Slot/ButtonStyle-Slot-Gadget-S.ButtonStyle-Slot-Gadget-S_C
    hoveredSlateSoundOverride {resourceObject=/Game/Sounds/Fort_UI_Sounds/FrontEnd/New_FrontEnd/Fortnite_UI_MainMenu_Tab_Hover_02_Cue}
    pressedSlateSoundOverride {resourceObject=&Fortnite_UI_MainMenu_Tab_Hover_02_Cue}
   }
   slot {class=OverlaySlot, properties {horizontalAlignment=HAlign_Center, verticalAlignment=VAlign_Center}}
  }
  Item {
   widgetClassPath=/Script/FortniteUI.FortMultiSizeItemCard
   bIsVariable=true
   namedSlotHost=Button
   widget=Item
   properties {bPItemCardSize=M, emptyItemType=AccountItem, renderTransform {angle=0, scale=<0.9 0.9>, shear=<0 0>, translation=<0 0>}}
  }
 ]
}
