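# Lobby player pad widget that shows a team member's two selected gadgets as stacked, scaled-down gadget buttons.
# Refresh reads the first two selected gadget items from team member info, casts each to a world item, and passes it to its button, or refreshes it empty otherwise.
# The buttons are hit-test invisible, so the widget relays hover through dispatchers and broadcasts left clicks with its stored PlayerIndex on the front-end context.
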WidgetBlueprint /Game/UI/Frontend/Lobby/PlayerPad/LobbyPlayerPadGadgets
graphs [
 EventGraph {nodes [
   Event OnMouseEnter()
   > Call On Gadgets Mouse Entered!(execute:Exec, self:Lobby Player Pad Gadgets Object Reference) -> (then:Exec) {is=title K2Node_CallDelegate}
   Event OnMouseLeave!() -> (OutputDelegate, then, MouseEvent:"Pointer Event Structure (by ref)") {is=title K2Node_Event}
   > Call On Gadgets Mouse Left!(execute:Exec, self:Lobby Player Pad Gadgets Object Reference) -> (then:Exec) {is=title K2Node_CallDelegate}
  ]}
 Initialize {nodes [
   n1 = Initialize!() -> (then, PlayerIndex:int) {is=title K2Node_FunctionEntry}
   > Set PlayerIndex(PlayerIndex:int=@n1) -> (Output_Get:int=0)
  ]}
 OnGadgetsMouseEntered {nodes [OnGadgetsMouseEntered() {is=title K2Node_FunctionEntry}]}
 OnGadgetsMouseLeft {nodes [OnGadgetsMouseLeft() {is=title K2Node_FunctionEntry}]}
 OnInputClicked {nodes [
   OnInputClicked() {is=title K2Node_FunctionEntry}
   > Call On Lobby Player Gadgets Clicked!(
    execute:Exec
    self:Fort Front End Context Object Reference=Get FortFrontEndContext!() -> (ReturnValue:Fort Front End Context Object Reference) {is=title K2Node_GetBlueprintContextEx}
    PlayerIndex:int=Get PlayerIndex() -> (PlayerIndex:int=0)
   ) -> (then:Exec) {is=title K2Node_CallDelegate}
  ]}
 OnMouseButtonDown {nodes [
   n1 = OnMouseButtonDown!() -> (then, MyGeometry:Geometry Structure, MouseEvent:"Pointer Event Structure (by ref)") {is=title K2Node_FunctionEntry}
   > n2 = Branch(Condition="Equal (Key)"!(
     A:Key Structure=Get Effecting Button!(Input:"Pointer Event Structure (by ref)"=@n1.MouseEvent) -> (ReturnValue:Key Structure) {is=title K2Node_CallFunction}
     B:Key Structure=LeftMouseButton
    ) -> (ReturnValue:bool))
   > n3 = OnInputClicked() {is=title K2Node_CallFunction}
   Return Node!(execute=[@n3.then, @n2.else], ReturnValue:Event Reply Structure=Handled())
  ]}
 Refresh {nodes [
   n1 = Refresh!() -> (then, TeamMemberInfo:Fort Team Member Info Structure) {is=title K2Node_FunctionEntry}
   > n2 = Sequence()
   n3 = Break Fort Team Member Info!(FortTeamMemberInfo:"Fort Team Member Info Structure (by ref)"=@n1) -> (
    MemberUniqueId:Unique Net Id Structure
    PartyLeaderUniqueId:Unique Net Id Structure
    ConsoleUniqueId:Unique Net Id Structure
    Platform:String
    PlayerName:Text
    bPartyLeader:bool=~
    bIsInZone:bool=~
    bHasBoostXp:bool=~
    bHasRestXp:bool=~
    BannerIconId:String
    BannerColorId:String
    bBattlePassPurchased:bool=~
    BattlePassLevel:int
    NumPlayersInParty:int
    PlayerIndex:int
    TeamAffiliation:Byte
    HeroClass:Text
    HeroLevel:Text
    HeroXP:int
    HeroItem:Fort Item Object Reference
    SelectedGadgetItems:Array of Fort Item Object References
    HeroIcon:Slate Brush Structure
    HomeBaseInfo:Fort Home Base Info Structure
   ) {is=title K2Node_BreakStruct}
   > n4 = Branch(Condition=Is Valid Index!(TargetArray:Array of Fort Item Object References=@n3.SelectedGadgetItems, IndexToTest:int) -> (ReturnValue:bool))
   > n5 = Cast To FortWorldItem!(execute, Object="Get (a copy)"(Array:Array of Fort Item Object References=@n3.SelectedGadgetItems) -> (Output:Fort Item Object Reference)) -> (then, CastFailed, AsFort World Item:Fort World Item Object Reference) {is=title K2Node_DynamicCast}
   > Refresh!(
    execute
    self:Lobby Gadget Button Object Reference=Get Gadget1() -> (Gadget1:Lobby Gadget Button Object Reference)
    WorldItem:Fort World Item Object Reference=@n5
   ) -> (then) {is=title K2Node_CallFunction}
   Refresh(execute=[@n5.CastFailed, @n4.else], self=Get Gadget1())
   n6 = Branch(execute=@n2.then_1, Condition=Is Valid Index(TargetArray=@n3.SelectedGadgetItems, IndexToTest=1))
   > n7 = Cast To FortWorldItem(Object="Get (a copy)"(Array=@n3.SelectedGadgetItems, Dimension 1=1))
   > Refresh(self=Get Gadget2() -> (Gadget2:Lobby Gadget Button Object Reference), WorldItem=@n7)
   Refresh(execute=[@n7.CastFailed, @n6.else], self=Get Gadget2())
  ]}
]
variables [
 OnGadgetsMouseEntered {}
 OnGadgetsMouseLeft {}
 PlayerIndex {type=integer}
]
widgets {
 info {inheritedWidgetCount=0, parentClass=&CommonUserWidget, rootWidgetClass=&VerticalBox, widgetCount=3}
 widgets [
  VerticalBox_0 {
   widgetClassPath=&VerticalBox
   widget=VerticalBox_0
   properties {visibility=Visible, renderTransform {angle=0, scale=<0.6 0.6>, shear=<0 0>, translation=<0 0>}}
  }
  Gadget1 {
   widgetClassPath=/Game/UI/Frontend/Lobby/LobbyGadgetButton
   bIsVariable=true
   parent=VerticalBox_0
   widget=Gadget1
   properties {visibility=HitTestInvisible}
   slot {class=VerticalBoxSlot, properties {verticalAlignment=VAlign_Bottom}}
  }
  Gadget2 {
   widgetClassPath=&LobbyGadgetButton
   bIsVariable=true
   parent=VerticalBox_0
   widget=Gadget2
   properties {visibility=HitTestInvisible}
   slot {class=VerticalBoxSlot, properties {verticalAlignment=VAlign_Bottom}}
  }
 ]
}
