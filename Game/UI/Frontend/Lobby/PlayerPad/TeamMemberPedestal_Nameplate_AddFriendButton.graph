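# "Add Friend!" CommonUI button widget for the team member pedestal nameplate in the lobby.
# Hover animations scale the label and fade the skewed yellow background to white and back, and clicking plays Reset.
# Construct sets height and shear for touch or non-touch input, while PreConstruct binds left-stick click in the Campaign subgame.
# Helpers set label text and input-icon visibility; PlayShine and Disabled animations exist but this graph never plays them.
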WidgetBlueprint /Game/UI/Frontend/Lobby/PlayerPad/TeamMemberPedestal_Nameplate_AddFriendButton
animations [
 OnHovered {
  binding_count=2
  length_seconds=0.3493
  object=&self:OnHovered
  sequence {bindings [
    CenterButtonTextWidget {
     id=#1
     possessed_class=CommonTextBlock
     tracks [{
       type=MovieScene2DTransformTrack
       sections [{
         type=MovieScene2DTransformSection
         channels [
          Translation.X {type=MovieSceneScriptingFloatChannel}
          Translation.Y {type=MovieSceneScriptingFloatChannel}
          Angle {type=MovieSceneScriptingFloatChannel}
          Scale.X {type=MovieSceneScriptingFloatChannel, keys [[0,1],[0.5637,0.9678],[2.8533,1.1208],[6.9863,1.0871]]}
          Scale.Y {type=MovieSceneScriptingFloatChannel, keys [[0,1],[0.5637,0.9678],[2.8533,1.1208],[6.9863,1.0871]]}
          Shear.X {type=MovieSceneScriptingFloatChannel}
          Shear.Y {type=MovieSceneScriptingFloatChannel}
         ]
        }]
      }]
    }
    Image_ButtonTop {
     id=#2
     possessed_class=Image
     tracks [{
       type=MovieSceneColorTrack
       display_name=Color and Opacity
       sections [{
         type=MovieSceneColorSection
         channels [
          Color.R {type=MovieSceneScriptingFloatChannel, keys [[0,0.9216],[6.986,1]]}
          Color.G {type=MovieSceneScriptingFloatChannel, const [1,2]}
          Color.B {type=MovieSceneScriptingFloatChannel, keys [[0,0],[6.986,1]]}
          Color.A {type=MovieSceneScriptingFloatChannel, const [1,2]}
         ]
        }]
      }]
    }
   ]}
 }
 PlayShine {
  binding_count=1
  length_seconds=1.0531
  object=&self:PlayShine
  sequence {bindings [CenterButtonTextWidget {
     id=#3
     possessed_class=CommonTextBlock
     tracks [{
       type=MovieSceneWidgetMaterialTrack
       display_name=Font.Outline Settings.OutlineMaterial
       sections [{
         type=MovieSceneParameterSection
         end_frame=21
         start_frame=0
         channels [ManualAnimateScalar {type=MovieSceneScriptingFloatChannel, keys [[0,0],[20.5053,0.25],[20.7433,0.2]]}]
        }]
      }]
    }]}
 }
 Disabled {
  binding_count=1
  length_seconds=5
  object=&self:Disabled
  sequence {bindings [InternalRootButton {
     id=#4
     possessed_class=CommonButtonInternalLegacy
     tracks [{
       type=MovieSceneWidgetMaterialTrack
       display_name=Style.Normal.Brush Material
       sections [{
         type=MovieSceneParameterSection
         end_frame=0
         start_frame=0
         channels [
          Color1R {type=MovieSceneScriptingFloatChannel, const [0.1094,1]}
          Color1G {type=MovieSceneScriptingFloatChannel, const [0.1094,1]}
          Color1B {type=MovieSceneScriptingFloatChannel, const [0.1094,1]}
          Color1A {type=MovieSceneScriptingFloatChannel, const [1,1]}
          Color2R {type=MovieSceneScriptingFloatChannel, const [0.1094,1]}
          Color2G {type=MovieSceneScriptingFloatChannel, const [0.1094,1]}
          Color2B {type=MovieSceneScriptingFloatChannel, const [0.1094,1]}
          Color2A {type=MovieSceneScriptingFloatChannel, const [1,1]}
          InnerColorR {type=MovieSceneScriptingFloatChannel, const [0.1094,1]}
          InnerColorG {type=MovieSceneScriptingFloatChannel, const [0.1094,1]}
          InnerColorB {type=MovieSceneScriptingFloatChannel, const [0.1094,1]}
          InnerColorA {type=MovieSceneScriptingFloatChannel, const [1,1]}
          ShineColorR {type=MovieSceneScriptingFloatChannel, const [0.1094,1]}
          ShineColorG {type=MovieSceneScriptingFloatChannel, const [0.1094,1]}
          ShineColorB {type=MovieSceneScriptingFloatChannel, const [0.1094,1]}
          ShineColorA {type=MovieSceneScriptingFloatChannel, const [1,1]}
         ]
        }]
      }]
    }]}
 }
 Reset {
  binding_count=2
  length_seconds=0
  object=&self:Reset
  sequence {bindings [
    CenterButtonTextWidget {
     id=#5
     possessed_class=CommonTextBlock
     tracks [{
       type=MovieScene2DTransformTrack
       sections [{
         type=MovieScene2DTransformSection
         end_frame=0
         start_frame=0
         channels [
          Translation.X {type=MovieSceneScriptingFloatChannel, const [0,1]}
          Translation.Y {type=MovieSceneScriptingFloatChannel, const [0,1]}
          Angle {type=MovieSceneScriptingFloatChannel, const [0,1]}
          Scale.X {type=MovieSceneScriptingFloatChannel, const [1,1]}
          Scale.Y {type=MovieSceneScriptingFloatChannel, const [1,1]}
          Shear.X {type=MovieSceneScriptingFloatChannel, const [-8,1]}
          Shear.Y {type=MovieSceneScriptingFloatChannel, const [0,1]}
         ]
        }]
      }]
    }
    Image_ButtonTop {
     id=#6
     possessed_class=Image
     tracks [{
       type=MovieSceneColorTrack
       display_name=Color and Opacity
       sections [{
         type=MovieSceneColorSection
         channels [
          Color.R {type=MovieSceneScriptingFloatChannel, const [0.9216,1]}
          Color.G {type=MovieSceneScriptingFloatChannel, const [1,1]}
          Color.B {type=MovieSceneScriptingFloatChannel, const [0,1]}
          Color.A {type=MovieSceneScriptingFloatChannel, const [1,1]}
         ]
        }]
      }]
    }
   ]}
 }
 OnUnhovered {
  binding_count=2
  length_seconds=0.35
  object=&self:OnUnhovered
  sequence {bindings [
    CenterButtonTextWidget {
     id=#1
     possessed_class=CommonTextBlock
     tracks [{
       type=MovieScene2DTransformTrack
       sections [{
         type=MovieScene2DTransformSection
         channels [
          Translation.X {type=MovieSceneScriptingFloatChannel}
          Translation.Y {type=MovieSceneScriptingFloatChannel}
          Angle {type=MovieSceneScriptingFloatChannel}
          Scale.X {type=MovieSceneScriptingFloatChannel, keys [[0,1.0871],[3,1.1208],[6,0.9678],[7,1]]}
          Scale.Y {type=MovieSceneScriptingFloatChannel, keys [[0,1.0871],[3,1.1208],[6,0.9678],[7,1]]}
          Shear.X {type=MovieSceneScriptingFloatChannel}
          Shear.Y {type=MovieSceneScriptingFloatChannel}
         ]
        }]
      }]
    }
    Image_ButtonTop {
     id=#2
     possessed_class=Image
     tracks [{
       type=MovieSceneColorTrack
       display_name=Color and Opacity
       sections [{
         type=MovieSceneColorSection
         channels [
          Color.R {type=MovieSceneScriptingFloatChannel, keys [[0,1],[6.986,0.9216]]}
          Color.G {type=MovieSceneScriptingFloatChannel, const [1,2]}
          Color.B {type=MovieSceneScriptingFloatChannel, keys [[0,1],[6.986,0]]}
          Color.A {type=MovieSceneScriptingFloatChannel, const [1,2]}
         ]
        }]
      }]
    }
   ]}
 }
]
graphs [
 EventGraph {nodes [
   n1 = SetInputActionVisibility!() -> (OutputDelegate, then, InVisibility:ESlateVisibility Enum=Visible) {is=title K2Node_CustomEvent}
   > SetVisibility(self=Get SB_InputAction() -> (SB_InputAction:Size Box Object Reference), InVisibility=@n1.InVisibility)
   Event On Clicked()
   > PlayAnimation(InAnimation=Get Reset() -> (Reset:Widget Animation Object Reference))
   Event Construct()
   > UpdateVisuals!(execute, self, IsUsingTouch:bool=IsUsingTouch(PlayerController=GetOwningPlayer())) -> (then)
   Event PreConstruct()
   > Branch(Condition=IsInSubGame!(SubGame:ESubGame Enum=Campaign) -> (ReturnValue:bool))
   > SetTriggeringInputAction!(
    execute
    self:Common Button Base Object Reference
    InputActionRow:"Data Table Row Handle Structure (by ref)"=Make Data Table Row Handle!(DataTable:Data Table Object Reference=&InputActionDataTable, RowName:Name=GenericLeftStickClick) -> (DataTableRowHandle:Data Table Row Handle Structure) {is=title K2Node_MakeStruct}
   ) -> (then)
   Event On Hovered()
   > PlayAnimationForward(InAnimation=Get OnHovered() -> (OnHovered:Widget Animation Object Reference))
   Event On Unhovered()
   > PlayAnimationForward(InAnimation=Get OnUnhovered() -> (OnUnhovered:Widget Animation Object Reference))
  ]}
 SetButtonText {nodes [
   n1 = SetButtonText!() -> (then, InText:Text) {is=title K2Node_FunctionEntry}
   > "SetText (Text)"(self=Get CenterButtonTextWidget() -> (CenterButtonTextWidget:Common Text Object Reference), InText=@n1)
  ]}
 UpdateVisuals {nodes [
   n1 = UpdateVisuals!() -> (then, IsUsingTouch:bool=~) {is=title K2Node_FunctionEntry}
   > n2 = Branch(Condition=@n1)
   > SetHeightOverride(self=Get SizeBox_0() -> (SizeBox_0:Size Box Object Reference), InHeightOverride=50)
   > SetRenderShear!(
    execute
    self:Widget Object Reference=Get Image_ButtonTop() -> (Image_ButtonTop:Image Object Reference)
    Shear:Vector 2D Structure=(X=-10,Y=0)
   ) -> (then)
   ClearHeightOverride(execute=@n2.else, self:Size Box Object Reference=Get SizeBox_0())
   > SetRenderShear(self=Get Image_ButtonTop(), Shear=(X=0,Y=0))
  ]}
]
variables []
widgets {
 info {inheritedWidgetCount=1, parentClass=&CommonButtonLegacy, rootWidgetClass=&Overlay, widgetCount=9}
 widgets [
  Overlay_1 {widgetClassPath=&Overlay, widget=Overlay_1}
  SizeBox_0 {
   widgetClassPath=&SizeBox
   bIsVariable=true
   parent=Overlay_1
   widget=SizeBox_0
   properties {bOverride_MinDesiredWidth=true, heightOverride=50, minDesiredWidth=103}
   slot {class=OverlaySlot}
  }
  Overlay_0 {widgetClassPath=&Overlay, parent=SizeBox_0, widget=Overlay_0, slot {class=SizeBoxSlot}}
  Overlay_2 {
   widgetClassPath=&Overlay
   parent=Overlay_0
   widget=Overlay_2
   slot {class=OverlaySlot, properties {horizontalAlignment=HAlign_Fill, verticalAlignment=VAlign_Fill}}
  }
  Image_ButtonTop {
   widgetClassPath=&Image
   bIsVariable=true
   parent=Overlay_2
   widget=Image_ButtonTop
   properties {
    brush {
     bIsDynamicallyLoaded=false
     drawAs=Image
     imageType=NoImage
     mirroring=NoMirror
     resourceName=None
     resourceObject=/Game/UI/Frontend/Showdown/Materials/M_UI_SkewBox_LButton
     tiling=NoTile
     imageSize=<32 32>
     margin=<ltrb 0 0 0 0>
     outlineSettings {
      bUseBrushTransparency=false
      roundingType=HalfHeightRadius
      width=0
      color {colorUseRule=UseColor_Specified, specifiedColor=<rgba 0 0 0 0>}
      cornerRadii=<0 0 0 0>
     }
     tintColor {colorUseRule=UseColor_Specified, specifiedColor=<rgba 1 1 1 1>}
     uVRegion {bIsValid=false, max=<0 0>, min=<0 0>}
    }
    colorAndOpacity=<rgba 0.9216 1 0 1>
   }
   slot {class=OverlaySlot, properties {horizontalAlignment=HAlign_Fill, verticalAlignment=VAlign_Fill}}
  }
  HorizontalBox_0 {
   widgetClassPath=&HorizontalBox
   parent=Overlay_0
   widget=HorizontalBox_0
   slot {class=OverlaySlot, properties {horizontalAlignment=HAlign_Center, verticalAlignment=VAlign_Fill, padding=<ltrb 20 0 20 0>}}
  }
  CenterButtonTextWidget {
   widgetClassPath=&CommonTextBlock
   bIsVariable=true
   parent=HorizontalBox_0
   widget=CenterButtonTextWidget
   properties {
    justification=Center
    scrollStyle=&TextScrollStyle-Base_C
    text="Add Friend!"
    textTransformPolicy=ToUpper
    colorAndOpacity {colorUseRule=UseColor_Specified, specifiedColor=<rgba 0 0 0 1>}
    font {
     bForceMonospaced=false
     bMaterialIsStencil=false
     fontMaterial=null
     fontObject=&BurbankSmall
     letterSpacing=0
     monospacedWidth=1
     size=16
     skewAmount=0
     typefaceFontName=Black
     outlineSettings {
      bApplyOutlineToDropShadows=false
      bMiteredCorners=false
      bSeparateFillAlpha=false
      outlineMaterial=null
      outlineSize=0
      outlineColor=<rgba 1 1 1 1>
     }
    }
    renderTransform {angle=0, scale=<1 1>, shear=<-5 0>, translation=<0 0>}
    shadowOffset=<0 0>
   }
   slot {class=HorizontalBoxSlot, properties {verticalAlignment=VAlign_Center, padding=<ltrb 10 0 10 0>}}
  }
  SB_InputAction {
   widgetClassPath=&SizeBox
   bIsVariable=true
   parent=Overlay_1
   widget=SB_InputAction
   properties {visibility=HitTestInvisible}
   slot {class=OverlaySlot, properties {padding=<ltrb -20 -10 0 0>}}
  }
  InputActionWidget {
   widgetClassPath=&CommonActionWidget
   bInherited=true
   bIsVariable=true
   parent=SB_InputAction
   widget=InputActionWidget
   properties {
    progressMaterialParam=Percentage
    visibility=HitTestInvisible
    progressMaterialBrush {
     bIsDynamicallyLoaded=false
     drawAs=Image
     imageType=NoImage
     mirroring=NoMirror
     resourceName=None
     resourceObject=/Game/UI/Materials/HollowCircularProgressBar
     tiling=NoTile
     imageSize=<36 36>
     margin=<ltrb 0 0 0 0>
     outlineSettings {
      bUseBrushTransparency=false
      roundingType=HalfHeightRadius
      width=0
      color {colorUseRule=UseColor_Specified, specifiedColor=<rgba 0 0 0 0>}
      cornerRadii=<0 0 0 0>
     }
     tintColor {colorUseRule=UseColor_Specified, specifiedColor=<rgba 1 1 0 1>}
     uVRegion {bIsValid=false, max=<0 0>, min=<0 0>}
    }
   }
   slot {class=SizeBoxSlot}
  }
 ]
}
