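# Lobby nameplate for a team member's pedestal, showing banner, level, name, power rating, ready status, voice and platform icons, and an add-friend button.
# It appears only for a valid member in hero-select, centered-lobby or multi-squad cameras.
# Parent-class events drive ready color and sound, friend button and Athena-only level, while input method adjusts prompts and touch layout.
# BannerSwap plays forward or reverse for Habanero state, resetting if interrupted.
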WidgetBlueprint /Game/UI/Frontend/Lobby/PlayerPad/TeamMemberPedestal_Nameplate
animations [
 On Ready Change Text {
  binding_count=1
  display_label=""
  length_seconds=0.4
  object=&self:On Ready Change Text
  sequence {
   bindings [Text_ReadyStatus {
     id=#1
     possessed_class=CommonTextBlock
     tracks [
      {
       type=MovieScene2DTransformTrack
       sections [{
         type=MovieScene2DTransformSection
         end_frame=40
         start_frame=0
         channels [
          Translation.X {type=MovieSceneScriptingFloatChannel}
          Translation.Y {type=MovieSceneScriptingFloatChannel}
          Angle {type=MovieSceneScriptingFloatChannel}
          Scale.X {type=MovieSceneScriptingFloatChannel, keys [[0,0.25],[14.9983,1.125],[25,1]]}
          Scale.Y {type=MovieSceneScriptingFloatChannel, keys [[0,2],[12.9983,0.75],[24.9983,1]]}
          Shear.X {type=MovieSceneScriptingFloatChannel}
          Shear.Y {type=MovieSceneScriptingFloatChannel}
         ]
        }]
      }
      {
       type=MovieSceneFloatTrack
       display_name=Render Opacity
       sections [{
         type=MovieSceneFloatSection
         end_frame=40
         start_frame=0
         channels [None {type=MovieSceneScriptingFloatChannel, keys [[4.9983,0],[19.9983,1]]}]
        }]
      }
     ]
    }]
   display_rate {numerator=100}
  }
 }
 ReadyTextColor {
  binding_count=1
  display_label=""
  length_seconds=0.0167
  object=&self:ReadyTextColor
  sequence {bindings [Text_ReadyStatus {
     id=#2
     possessed_class=CommonTextBlock
     tracks [{
       type=MovieSceneColorTrack
       display_name=Color and Opacity
       sections [{
         type=MovieSceneColorSection
         end_frame=0
         start_frame=0
         channels [
          Color.R {type=MovieSceneScriptingFloatChannel, const [0.0343,1]}
          Color.G {type=MovieSceneScriptingFloatChannel, const [1,1]}
          Color.B {type=MovieSceneScriptingFloatChannel, const [0.227,1]}
          Color.A {type=MovieSceneScriptingFloatChannel, const [1,1]}
         ]
        }]
      }]
    }]}
 }
 NotReadyTextColor {
  binding_count=1
  display_label=""
  length_seconds=0.01
  object=&self:NotReadyTextColor
  sequence {bindings [Text_ReadyStatus {
     id=#2
     possessed_class=CommonTextBlock
     tracks [{
       type=MovieSceneColorTrack
       display_name=Color and Opacity
       sections [{
         type=MovieSceneColorSection
         end_frame=0
         start_frame=0
         channels [
          Color.R {type=MovieSceneScriptingFloatChannel, const [1,1]}
          Color.G {type=MovieSceneScriptingFloatChannel, const [0.17,1]}
          Color.B {type=MovieSceneScriptingFloatChannel, const [0.15,1]}
          Color.A {type=MovieSceneScriptingFloatChannel, const [1,1]}
         ]
        }]
      }]
    }]}
 }
 BannerSwap {
  binding_count=10
  length_seconds=0.3
  object=&self:BannerSwap
  sequence {bindings [
    Spacer_307 {
     id=#3
     possessed_class=Spacer
     tracks [{
       type=MovieSceneDoubleVectorTrack
       display_name=Size
       sections [{
         type=MovieSceneDoubleVectorSection
         end_frame=6
         start_frame=0
         channels [
          Vector.X {type=MovieSceneScriptingDoubleChannel, keys [[0,40],[6,92]]}
          Vector.Y {type=MovieSceneScriptingDoubleChannel}
         ]
        }]
      }]
    }
    OverlaySlot_0 {
     id=#4
     parent_id=#5
     possessed_class=OverlaySlot
     tracks [{
       type=MovieSceneMarginTrack
       display_name=Padding
       sections [{
         type=MovieSceneMarginSection
         end_frame=6
         start_frame=0
         channels [
          Left {type=MovieSceneScriptingFloatChannel, keys [[0,-32],[6,0]]}
          Top {type=MovieSceneScriptingFloatChannel}
          Right {type=MovieSceneScriptingFloatChannel}
          Bottom {type=MovieSceneScriptingFloatChannel}
         ]
        }]
      }]
    }
    OverlaySlot_0 {
     id=#6
     parent_id=#7
     possessed_class=OverlaySlot
     tracks [{
       type=MovieSceneMarginTrack
       display_name=Padding
       sections [{
         type=MovieSceneMarginSection
         end_frame=0
         start_frame=0
         channels [
          Left {type=MovieSceneScriptingFloatChannel}
          Top {type=MovieSceneScriptingFloatChannel}
          Right {type=MovieSceneScriptingFloatChannel, const [48,1]}
          Bottom {type=MovieSceneScriptingFloatChannel}
         ]
        }]
      }]
    }
    OverlaySlot_3 {
     id=#8
     parent_id=#9
     possessed_class=OverlaySlot
     tracks [{
       type=MovieSceneMarginTrack
       display_name=Padding
       sections [{
         type=MovieSceneMarginSection
         end_frame=6
         start_frame=0
         channels [
          Left {type=MovieSceneScriptingFloatChannel}
          Top {type=MovieSceneScriptingFloatChannel, keys [[0,14],[6,0]]}
          Right {type=MovieSceneScriptingFloatChannel}
          Bottom {type=MovieSceneScriptingFloatChannel}
         ]
        }]
      }]
    }
    VerticalBox_2 {id=#7, possessed_class=VerticalBox, tracks []}
    VoiceIcon_Status {
     id=#10
     possessed_class=FortVoiceChatStatusIconProxyAnim
     tracks [
      {
       type=MovieSceneFloatTrack
       display_name=Render Opacity
       sections [{
         type=MovieSceneFloatSection
         end_frame=6
         start_frame=0
         channels [None {type=MovieSceneScriptingFloatChannel, keys [[0,1],[2,0],[4,0],[6,1]]}]
        }]
      }
      {
       type=MovieScene2DTransformTrack
       sections [{
         type=MovieScene2DTransformSection
         end_frame=6
         start_frame=2
         channels [
          Translation.X {type=MovieSceneScriptingFloatChannel}
          Translation.Y {type=MovieSceneScriptingFloatChannel}
          Angle {type=MovieSceneScriptingFloatChannel}
          Scale.X {type=MovieSceneScriptingFloatChannel, keys [[2,1],[4,0.6],[6,0.7]]}
          Scale.Y {type=MovieSceneScriptingFloatChannel, keys [[2,1],[4,0.6],[6,0.7]]}
          Shear.X {type=MovieSceneScriptingFloatChannel}
          Shear.Y {type=MovieSceneScriptingFloatChannel}
         ]
        }]
      }
     ]
    }
    OverlaySlot_1 {
     id=#11
     parent_id=#10
     possessed_class=OverlaySlot
     tracks [
      {
       type=MovieSceneByteTrack
       display_name=Horizontal Alignment
       sections [{type=MovieSceneByteSection, end_frame=4, start_frame=0, channels [None {type=MovieSceneScriptingByteChannel, keys [[0,3],[2,3],[4,1]]}]}]
      }
      {
       type=MovieSceneByteTrack
       display_name=Vertical Alignment
       sections [{type=MovieSceneByteSection, end_frame=4, start_frame=0, channels [None {type=MovieSceneScriptingByteChannel, keys [[0,1],[2,1],[4,3]]}]}]
      }
      {
       type=MovieSceneMarginTrack
       display_name=Padding
       sections [{
         type=MovieSceneMarginSection
         end_frame=4
         start_frame=4
         channels [
          Left {type=MovieSceneScriptingFloatChannel, const [-32,1]}
          Top {type=MovieSceneScriptingFloatChannel}
          Right {type=MovieSceneScriptingFloatChannel}
          Bottom {type=MovieSceneScriptingFloatChannel}
         ]
        }]
      }
     ]
    }
    HabaneroWidget {
     id=#5
     possessed_class=WBP_Habanero_Icon_C
     tracks [
      {
       type=MovieSceneFloatTrack
       display_name=Render Opacity
       sections [{type=MovieSceneFloatSection, end_frame=6, start_frame=4, channels [None {type=MovieSceneScriptingFloatChannel, keys [[4,0],[6,1]]}]}]
      }
      {
       type=MovieSceneByteTrack
       display_name=Visibility
       sections [{
         type=MovieSceneByteSection
         end_frame=6
         start_frame=0
         channels [None {type=MovieSceneScriptingByteChannel, keys [[0,1],[2,1],[3,3],[6,3]]}]
        }]
      }
     ]
    }
    Widget_PlayerLevel {
     id=#12
     possessed_class=SizeBox
     tracks [{
       type=MovieSceneFloatTrack
       display_name=Render Opacity
       sections [{
         type=MovieSceneFloatSection
         end_frame=6
         start_frame=0
         channels [None {type=MovieSceneScriptingFloatChannel, thinned_from=4, keys [[0,1],[3,1],[6,0]]}]
        }]
      }]
    }
    Banner_Containment {
     id=#9
     possessed_class=Overlay
     tracks [{
       type=MovieScene2DTransformTrack
       sections [{
         type=MovieScene2DTransformSection
         end_frame=6
         start_frame=0
         channels [
          Translation.X {type=MovieSceneScriptingFloatChannel}
          Translation.Y {type=MovieSceneScriptingFloatChannel}
          Angle {type=MovieSceneScriptingFloatChannel}
          Scale.X {type=MovieSceneScriptingFloatChannel, keys [[0,1],[6,0.6]]}
          Scale.Y {type=MovieSceneScriptingFloatChannel, keys [[0,1],[6,0.6]]}
          Shear.X {type=MovieSceneScriptingFloatChannel}
          Shear.Y {type=MovieSceneScriptingFloatChannel}
         ]
        }]
      }]
    }
   ]}
 }
]
graphs [
 EventGraph {nodes [
   n1 = OnFrontendCameraChanged!() -> (OutputDelegate, then, NewCamera:EFrontEndCamera Enum=Invalid, OldCamera:EFrontEndCamera Enum=Invalid) {is=title K2Node_CustomEvent}
   n2 = NOT Boolean(A=IsCurrentMemberValid!(self:Fort Team Member Entry Base Object Reference) -> (ReturnValue:bool))
   n3 = NOT Boolean(A=OR Boolean!(
     A="Equal (Enum)"(A:EFrontEndCamera Enum=@n1.NewCamera, B:EFrontEndCamera Enum=HeroSelect)
     B="Equal (Enum)"(A=@n1.NewCamera, B=LobbyCentered)
     C:bool="Equal (Enum)"(A=@n1.NewCamera, B=MultipleSquads)
    ) -> (ReturnValue))
   > SetVisibility(InVisibility=Select(Option 0=SelfHitTestInvisible, Option 1=Hidden, Index=OR Boolean!(A=@n2, B=@n3) -> (ReturnValue)))
   n4 = OnInputChanged!() -> (OutputDelegate, then, bNewInputType:ECommonInputType Enum) {is=title K2Node_CustomEvent}
   > Update Add Friend Button Visuals!(
    execute
    self
    IsUsingTouch:bool=Select!(MouseAndKeyboard:bool=false, Gamepad:bool=false, Touch:bool=true, Count:bool=false, Index:ECommonInputType Enum=@n4.bNewInputType) -> (ReturnValue:bool)
   ) -> (then)
   HandleInputMethodChanged!() -> (OutputDelegate, then, bNewInputType:ECommonInputType Enum) {is=title K2Node_CustomEvent}
   > UpdateInputActionVisibility()
   Event Construct()
   n5 = Create Event()
   n6 = Get FortFrontEndContext!() -> (ReturnValue:Fort Front End Context Object Reference) {is=title K2Node_GetBlueprintContextEx}
   > Bind Event to On Front End Camera Changed(self:Fort Front End Context Object Reference=@n6, Delegate=@n5) {is=title K2Node_AddDelegate}
   > OnFrontendCameraChanged!(
    execute
    self
    NewCamera:EFrontEndCamera Enum=GetFrontEndCamera!(self:Fort Front End Context Object Reference=@n6) -> (ReturnValue:EFrontEndCamera Enum=Invalid)
    OldCamera:EFrontEndCamera Enum=Get DefaultOldCameraValue() -> (DefaultOldCameraValue:EFrontEndCamera Enum=Invalid)
   ) -> (then) {is=title K2Node_CallFunction}
   n7 = Create Event()
   n8 = Get CommonInputSubsystem()
   > Bind Event to On Input Method Changed(self=@n8, Delegate=@n7)
   > UpdateInputActionVisibility()
   > ResetHabaneroAnimation!(execute, self, HabaneroState:bool=false) -> (then)
   > n9 = Branch(Condition=IsMobileGame())
   n10 = Get ScaleBox_0() -> (ScaleBox_0:Scale Box Object Reference)
   > SetStretch!(execute, self:Scale Box Object Reference=@n10, InStretch:EStretch Enum=Fill) -> (then)
   SetStretch(execute=@n9.else, self=@n10, InStretch=ScaleToFit)
   Event Destruct()
   > n11 = Is Valid(InputObject=@n6)
   > n12 = Unbind Event from On Front End Camera Changed(self:Fort Front End Context Object Reference=@n6, Delegate=@n5) {is=title K2Node_RemoveDelegate}
   Is Valid(exec=[@n11.Is Not Valid, @n12.then], InputObject=@n8)
   > Unbind Event from On Input Method Changed(self=@n8, Delegate=@n7)
   Event BP_OnTeamMemberEstablished!() -> (OutputDelegate, then, bHasValidMember:bool) {is=title K2Node_Event}
   > UpdateInputActionVisibility()
   n13 = Event OnPlayReadyAudio!() -> (OutputDelegate, then, bIsReadyForMatchmaking:bool) {is=title K2Node_Event}
   > PlaySound2D(Sound=Select!(
     Option 0:Sound Base Object Reference=Get UnReadySound() -> (UnReadySound:Sound Base Object Reference)
     Option 1:Sound Base Object Reference=Get ReadySound() -> (ReadySound:Sound Base Object Reference)
     Index:bool=@n13.bIsReadyForMatchmaking
    ) -> (ReturnValue:Sound Base Object Reference))
   n14 = Event OnCurrentSubGameChanged!() -> (OutputDelegate, then, NewSubGame:ESubGame Enum=Campaign) {is=title K2Node_Event}
   > SetVisibility(
    self=Get Widget_PlayerLevel() -> (Widget_PlayerLevel:Size Box Object Reference)
    InVisibility=Select(
     Option 0:ESlateVisibility Enum=Collapsed
     Option 1:ESlateVisibility Enum=Visible
     Index="Equal (Enum)"(A:ESubGame Enum=@n14.NewSubGame, B:ESubGame Enum=Athena)
    ) -> (ReturnValue:ESlateVisibility Enum)
   )
   n15 = Event OnReadyStatusChanged!() -> (OutputDelegate, then, GameReadiness:EGameReadiness Enum=NotReady, bIsDownloading:bool) {is=title K2Node_Event}
   n16 = Get ReadyColor() -> (ReadyColor:Linear Color Structure)
   n17 = Get NotReadyColor() -> (NotReadyColor:Linear Color Structure)
   > SetColorAndOpacity(
    self=Get Text_ReadyStatus() -> (Text_ReadyStatus:Common Text Object Reference)
    InColorAndOpacity=Make Slate Color(SpecifiedColor=Select!(
      NotReady:Linear Color Structure=@n17
      Ready:Linear Color Structure=@n16
      Reevaluating:Linear Color Structure
      SittingOut:Linear Color Structure=@n17
      Away:Linear Color Structure=@n17
      Index:EGameReadiness Enum=@n15.GameReadiness
     ) -> (ReturnValue:Linear Color Structure))
   )
   > Branch(Condition=NOT Boolean(A=@n15.bIsDownloading))
   > PlayAnimation(InAnimation=Get On Ready Change Text() -> (On Ready Change Text:Widget Animation Object Reference))
   Event OnInitialized()
   > Update Add Friend Button Visuals(IsUsingTouch=IsUsingTouch(PlayerController=GetOwningPlayer()))
   > Bind Event to On Input Method Changed(self=Get CommonInputSubsystem(), Delegate=@n4)
   n18 = Event OnFriendStatusUpdated!() -> (OutputDelegate, then, FriendRequestStatus:EFortFriendRequestStatus Enum=None) {is=title K2Node_Event}
   n19 = Select!(
    None:ESlateVisibility Enum=Visible
    Accepted:ESlateVisibility Enum=Collapsed
    PendingReceived:ESlateVisibility Enum=Visible
    PendingSent:ESlateVisibility Enum=Collapsed
    Index:EFortFriendRequestStatus Enum=@n18.FriendRequestStatus
   ) -> (ReturnValue:ESlateVisibility Enum)
   n20 = Get Button_AddFriend() -> (Button_AddFriend:Common Button Legacy Object Reference)
   > SetVisibility(self=@n20, InVisibility=@n19)
   > Branch(Condition="Equal (Enum)"(A:ESlateVisibility Enum=@n19, B:ESlateVisibility Enum=Visible))
   > n21 = Cast To TeamMemberPedestal_Nameplate_AddFriendButton!(execute, Object=@n20) -> (then, CastFailed, AsTeam Member Pedestal Nameplate Add Friend Button:Team Member Pedestal Nameplate Add Friend Button Object Reference) {is=title K2Node_DynamicCast}
   n22 = Get CenterButtonTextWidget!(self:Team Member Pedestal Nameplate Add Friend Button Object Reference=@n21) -> (CenterButtonTextWidget:Common Text Object Reference)
   > Is Valid(InputObject=@n22)
   n23 = GetEmptyText!() -> (ReturnValue:Text)
   > "SetText (Text)"(
    self=@n22
    InText=Select(
     None:"Text (by ref)"=Get AddFriendText() -> (AddFriendText:Text)
     Accepted:"Text (by ref)"=@n23
     PendingReceived:"Text (by ref)"=Get AcceptFriendRequestText() -> (AcceptFriendRequestText:Text)
     PendingSent:"Text (by ref)"=@n23
     Index=@n18.FriendRequestStatus
    ) -> (ReturnValue:"Text (by ref)")
   )
   n24 = Event OnIsPlayingHabaneroChanged!() -> (OutputDelegate, then, bPlayingHabanero:bool) {is=title K2Node_Event}
   > n25 = Branch(Condition=IsAnimationPlaying(InAnimation=Get BannerSwap() -> (BannerSwap:Widget Animation Object Reference)))
   > ResetHabaneroAnimation(HabaneroState=@n24.bPlayingHabanero)
   n26 = Branch(
    execute=@n25.else
    Condition="Equal (Boolean)"(A:bool=@n24.bPlayingHabanero, B:bool=Get HabaneroState() -> (HabaneroState:bool)) -> (ReturnValue:bool=~) {is=title K2Node_PromotableOperator}
   )
   n27 = "Play Animation with Finished event (legacy)"!(
    execute:Exec=@n26.else
    Widget:User Widget Object Reference=Self-Reference()
    InAnimation:Widget Animation Object Reference=Get BannerSwap()
    StartAtTime:float=0
    NumLoopsToPlay:int=1
    PlayMode:EUMGSequencePlayMode Enum=Select!(Option 0:EUMGSequencePlayMode Enum=Reverse, Option 1:EUMGSequencePlayMode Enum, Index:bool=@n24.bPlayingHabanero) -> (ReturnValue:EUMGSequencePlayMode Enum)
    PlaybackSpeed:float=1
   ) -> (then:Exec, Result:UMGSequence Player Object Reference, Finished:Exec) {is=title K2Node_PlayAnimation}
   Set HabaneroState(execute=@n27.Finished, HabaneroState:bool=@n24.bPlayingHabanero) -> (Output_Get:bool)
  ]}
 GetInputActionVisibility {nodes [
   GetInputActionVisibility() {is=title K2Node_FunctionEntry}
   n1 = IsUsingGamepad!(self:Fort Global UIContext Object Reference=Get FortGlobalUIContext!() -> (ReturnValue:Fort Global UIContext Object Reference) {is=title K2Node_GetBlueprintContextEx}) -> (ReturnValue:bool)
   n2 = Get bHovered() -> (bHovered:bool)
   > Return Node!(
    execute
    ReturnValue:ESlateVisibility Enum=Select(Option 0:ESlateVisibility Enum=Hidden, Option 1:ESlateVisibility Enum=SelfHitTestInvisible, Index=AND Boolean(A=@n1, B=@n2)) -> (ReturnValue:ESlateVisibility Enum)
   )
  ]}
 ResetHabaneroAnimation {nodes [
   n1 = ResetHabaneroAnimation!() -> (then, HabaneroState:bool=~) {is=title K2Node_FunctionEntry}
   n2 = Get BannerSwap()
   > StopAnimation(InAnimation=@n2)
   > Set HabaneroState(HabaneroState=@n1)
   > PlayAnimation(
    InAnimation=@n2
    StartAtTime=GetEndTime(self=@n2)
    PlayMode=Select(Option 0:EUMGSequencePlayMode Enum=Reverse, Option 1:EUMGSequencePlayMode Enum, Index=@n1) -> (ReturnValue:EUMGSequencePlayMode Enum)
   )
  ]}
 Update Add Friend Button Visuals {nodes [
   Update Add Friend Button Visuals!() -> (then, IsUsingTouch:bool=~) {is=title K2Node_FunctionEntry}
   > n1 = Cast To TeamMemberPedestal_Nameplate_AddFriendButton(Object=Get Button_AddFriend())
   n2 = Get IsUsingTouch() -> (IsUsingTouch:bool)
   > UpdateVisuals!(execute, self:Team Member Pedestal Nameplate Add Friend Button Object Reference=@n1, IsUsingTouch:bool=@n2) -> (then)
   > n3 = Branch(Condition=@n2)
   n4 = SlotAsOverlaySlot(Widget=Get TeamMemberPedestal_ActivatableContainer() -> (TeamMemberPedestal_ActivatableContainer:Team Member Pedestal Activatable Container Object Reference))
   > SetPadding!(
    execute
    self:Overlay Slot Object Reference=@n4
    InPadding_Left:float=0
    InPadding_Top:float=-60
    InPadding_Right:float=0
    InPadding_Bottom:float=0
   ) -> (then)
   SetPadding(execute=@n3.else, self=@n4, InPadding_Left=0, InPadding_Top=-40, InPadding_Right=0, InPadding_Bottom=0)
  ]}
 UpdateInputActionVisibility {nodes [
   UpdateInputActionVisibility() {is=title K2Node_FunctionEntry}
   n1 = GetInputActionVisibility!() -> (ReturnValue:ESlateVisibility Enum=Visible) {is=title K2Node_CallFunction}
   > SetVisibility(self=Get SB_InputAction() -> (SB_InputAction:Size Box Object Reference), InVisibility=@n1)
   > n2 = Cast To TeamMemberPedestal_Nameplate_AddFriendButton(Object=Get Button_AddFriend())
   > SetInputActionVisibility!(execute, self:Team Member Pedestal Nameplate Add Friend Button Object Reference=@n2, InVisibility:ESlateVisibility Enum=@n1) -> (then)
  ]}
]
variables [
 AcceptedInviteText {type=string, default="ACCEPTED INVITE!"}
 AcceptFriendRequestText {type=string, default="ACCEPT REQUEST!"}
 AddFriendText {type=string, default="ADD FRIEND!"}
 bHovered {type=boolean}
 DefaultOldCameraValue {type=EFrontEndCamera, default=Invalid}
 False {type=ESlateVisibility, default=Hidden}
 HabaneroState {type=boolean}
 NotReadyColor {type=LinearColor, default=<rgba 1 0.1793 0.1562 1>}
 ReadyColor {type=LinearColor, default=<rgba 0.0343 1 0.227 1>}
 ReadySound {type=SoundBase, default=/Game/Sounds/Fort_UI_Sounds/FrontEnd/New_FrontEnd/Fortnite_UI_ReadyUp_01_Cue}
 SentInviteText {type=string, default="SENT INVITE!"}
 TeamMemberInfo {
  type=FortTeamMemberInfo
  default {
   bBattlePassPurchased=false
   bHasBoostXp=false
   bHasRestXp=false
   bIsInZone=false
   bPartyLeader=false
   bannerColorId=""
   bannerIconId=""
   battlePassLevel=0
   heroClass=""
   heroItem=null
   heroLevel=""
   heroXP=-1
   numPlayersInParty=0
   playerIndex=-1
   playerName=""
   teamAffiliation=255
   heroIcon {
    bIsDynamicallyLoaded=false
    drawAs=Image
    imageType=NoImage
    mirroring=NoMirror
    resourceName=None
    resourceObject=null
    tiling=NoTile
    imageSize=<32 32>
    margin=<ltrb 0 0 0 0>
    outlineSettings {
     bUseBrushTransparency=false
     roundingType=HalfHeightRadius
     width=0
     color {colorUseRule=UseColor_Specified, specifiedColor=<rgba 0 0 0 0>}
     cornerRadii=<0 0 0 1>
    }
    tintColor {colorUseRule=UseColor_Specified, specifiedColor=<rgba 1 1 1 1>}
    uVRegion {bIsValid=false, max=<0 0>, min=<0 0>}
   }
   homeBaseInfo {name="", bannerColorId="", bannerIconId="", validData=false}
   selectedGadgetItems []
  }
 }
 UnReadySound {type=SoundBase, default=/Game/Sounds/Fort_UI_Sounds/FrontEnd/New_FrontEnd/Fortnite_UI_UnReady_01_Cue}
]
widgets {
 info {
  inheritedWidgetCount=17
  namedSlotCount=2
  parentClass=/Script/FortniteUI.FortTeamMemberPedestalNameplate
  rootWidgetClass=/Game/Athena/HUD/FortHUDElementWrapper.FortHUDElementWrapper_C
  widgetCount=34
 }
 widgets [
  FortHUDElementWrapper {
   widgetClassPath=/Game/Athena/HUD/FortHUDElementWrapper
   bIsVariable=true
   widget=FortHUDElementWrapper
   properties {elementTag {tagName=HUD}}
  }
  Overlay_1 {widgetClassPath=&Overlay, namedSlotHost=FortHUDElementWrapper, widget=Overlay_1}
  HorizontalBox_1 {widgetClassPath=&HorizontalBox, parent=Overlay_1, widget=HorizontalBox_1, slot {class=OverlaySlot}}
  Overlay_0 {
   widgetClassPath=&Overlay
   parent=HorizontalBox_1
   widget=Overlay_0
   slot {class=HorizontalBoxSlot, properties {horizontalAlignment=HAlign_Right, verticalAlignment=VAlign_Top}}
  }
  Banner_Containment {
   widgetClassPath=&Overlay
   parent=Overlay_0
   widget=Banner_Containment
   properties {renderTransformPivot=<1 0>}
   slot {class=OverlaySlot, properties {horizontalAlignment=HAlign_Right, padding=<ltrb 0 14 0 0>}}
  }
  Banner_MemberBanner {
   widgetClassPath=/Script/FortniteUI.FortPlayerBanner
   bInherited=true
   bIsVariable=true
   parent=Banner_Containment
   widget=Banner_MemberBanner
   properties {bShowBorder=false, desiredSize=<37 44>, renderTransform {angle=0, scale=<1 1>, shear=<0 0>, translation=<0 2>}}
   slot {class=OverlaySlot}
  }
  Widget_PlayerLevel {
   widgetClassPath=&SizeBox
   bIsVariable=true
   parent=Banner_Containment
   widget=Widget_PlayerLevel
   properties {bOverride_WidthOverride=true, widthOverride=34}
   slot {class=OverlaySlot, properties {verticalAlignment=VAlign_Bottom, padding=<ltrb 0 30 0 0>}}
  }
  ScaleBox_0 {
   widgetClassPath=&ScaleBox
   bIsVariable=true
   parent=Widget_PlayerLevel
   widget=ScaleBox_0
   properties {stretch=Fill}
   slot {class=SizeBoxSlot}
  }
  Text_PlayerLevel {
   widgetClassPath=&CommonTextBlock
   bInherited=true
   bIsVariable=true
   parent=ScaleBox_0
   widget=Text_PlayerLevel
   properties {
    text=888
    font {
     bForceMonospaced=false
     bMaterialIsStencil=false
     fontMaterial=null
     fontObject=&BurbankSmall
     letterSpacing=0
     monospacedWidth=1
     size=18
     skewAmount=0
     typefaceFontName=Black
     outlineSettings {
      bApplyOutlineToDropShadows=true
      bMiteredCorners=false
      bSeparateFillAlpha=false
      outlineMaterial=null
      outlineSize=1
      outlineColor=<rgba 0.0208 0.0208 0.0208 0.5>
     }
    }
    renderTransform {angle=0, scale=<1 1>, shear=<0 0>, translation=<-21 0>}
    shadowColorAndOpacity=<rgba 0 0 0 1>
    shadowOffset=<1.2 1.2>
   }
   slot {class=ScaleBoxSlot}
  }
  Image_LeaderIcon {
   widgetClassPath=&Image
   bInherited=true
   bIsVariable=true
   parent=Banner_Containment
   widget=Image_LeaderIcon
   properties {
    visibility=HitTestInvisible
    brush {
     bIsDynamicallyLoaded=false
     drawAs=Image
     imageType=NoImage
     mirroring=NoMirror
     resourceName=None
     resourceObject=/Game/UI/Foundation/Banner/M-CrownShine
     tiling=NoTile
     imageSize=<37 30>
     margin=<ltrb 0 0 0 0>
     outlineSettings {
      bUseBrushTransparency=false
      roundingType=HalfHeightRadius
      width=0
      color {colorUseRule=UseColor_Specified, specifiedColor=<rgba 0 0 0 0>}
      cornerRadii=<0 0 0 0>
     }
     tintColor {colorUseRule=UseColor_Specified, specifiedColor=<rgba 1 1 1 1>}
     uVRegion {bIsValid=false, max=<0 0>, min=<0 0>}
    }
    renderTransform {angle=0, scale=<1 1>, shear=<0 0>, translation=<0 -23>}
   }
   slot {class=OverlaySlot, properties {horizontalAlignment=HAlign_Center}}
  }
  SB_InputAction {
   widgetClassPath=&SizeBox
   bIsVariable=true
   parent=Overlay_0
   widget=SB_InputAction
   properties {visibility=Hidden}
   slot {class=OverlaySlot, properties {padding=<ltrb 0 0 0 8>}}
  }
  InputActionWidget {
   widgetClassPath=&CommonActionWidget
   bInherited=true
   bIsVariable=true
   parent=SB_InputAction
   widget=InputActionWidget
   properties {
    visibility=SelfHitTestInvisible
    inputActions [{dataTable=&InputActionDataTable, rowName=Select}]
    renderTransform {angle=0, scale=<1 1>, shear=<0 0>, translation=<-40 0>}
   }
   slot {class=SizeBoxSlot}
  }
  FortVisualAttachment_0 {
   widgetClassPath=&FortVisualAttachment
   parent=Overlay_0
   widget=FortVisualAttachment_0
   properties {widthOverride=64, contentAnchor=<0.5 0>}
   slot {class=OverlaySlot, properties {verticalAlignment=VAlign_Bottom}}
  }
  HorizontalBox_0 {
   widgetClassPath=&HorizontalBox
   parent=FortVisualAttachment_0
   widget=HorizontalBox_0
   slot {class=SizeBoxSlot, properties {horizontalAlignment=HAlign_Right}}
  }
  SizeBox_HotfixControlled {
   widgetClassPath=&SizeBox
   bInherited=true
   parent=HorizontalBox_0
   widget=SizeBox_HotfixControlled
   slot {class=HorizontalBoxSlot}
  }
  Image_PlatformIcon {
   widgetClassPath=&Image
   bInherited=true
   bIsVariable=true
   parent=SizeBox_HotfixControlled
   widget=Image_PlatformIcon
   slot {class=SizeBoxSlot}
  }
  Spacer_118 {widgetClassPath=&Spacer, parent=HorizontalBox_0, widget=Spacer_118, properties {size=<8 1>}, slot {class=HorizontalBoxSlot}}
  SizeBox_1 {
   widgetClassPath=&SizeBox
   parent=HorizontalBox_0
   widget=SizeBox_1
   properties {bOverride_HeightOverride=true, bOverride_WidthOverride=true, heightOverride=32, widthOverride=32}
   slot {class=HorizontalBoxSlot}
  }
  InputIndicator_ActiveInputType {
   widgetClassPath=/Game/Athena/UI/Matchmaking/MatchmakingInputIndicator
   bInherited=true
   bIsVariable=true
   parent=SizeBox_1
   widget=InputIndicator_ActiveInputType
   properties {bShowRemoteInputOnlyWhenDifferentFromDefault=true, inputSource=Remote}
   slot {class=SizeBoxSlot}
  }
  Spacer_Anim {
   widgetClassPath=&Spacer
   parent=Overlay_0
   widget=Spacer_Anim
   properties {size=<40 1>}
   slot {class=OverlaySlot, properties {horizontalAlignment=HAlign_Right}}
  }
  HabaneroWidget {
   widgetClassPath=/Game/UI/Frontend/Lobby/Habanero/WBP_Habanero_Icon
   bInherited=true
   bIsVariable=true
   parent=Overlay_0
   widget=HabaneroWidget
   properties {renderOpacity=0, visibility=Hidden, iconMinSize=<64 64>}
   slot {class=OverlaySlot, properties {padding=<ltrb -32 0 0 0>}}
  }
  Overlay_3 {
   widgetClassPath=&Overlay
   parent=HorizontalBox_1
   widget=Overlay_3
   slot {class=HorizontalBoxSlot, properties {verticalAlignment=VAlign_Top}}
  }
  VerticalBox_2 {
   widgetClassPath=&VerticalBox
   parent=Overlay_3
   widget=VerticalBox_2
   slot {class=OverlaySlot, properties {padding=<ltrb 8 0 48 0>}}
  }
  Text_SocialName {
   widgetClassPath=/Script/FortniteUI.FortSocialNameTextBlock
   bInherited=true
   bIsVariable=true
   parent=VerticalBox_2
   widget=Text_SocialName
   properties {
    bOverrideDefaultStyle=true
    bWrapToNextLine=true
    minDesiredSize=120
    scrollStyle=/FortUILibrary/Styles/Typography/Foundation/TextScrollStyle-NoFade.TextScrollStyle-NoFade_C
    text="User Name<PC/>\r\n<text color=\"1DEFA9FF\" size=\"12\">(NickName)</>"
    defaultTextStyleOverride {
     overflowPolicy=Clip
     transformPolicy=None
     colorAndOpacity {colorUseRule=UseColor_Specified, specifiedColor=<rgba 1 1 1 1>}
     font {
      bForceMonospaced=false
      bMaterialIsStencil=false
      fontMaterial=null
      fontObject=&HeadingNow
      letterSpacing=0
      monospacedWidth=1
      size=18
      skewAmount=0
      typefaceFontName=75 Medium
      outlineSettings {
       bApplyOutlineToDropShadows=false
       bMiteredCorners=false
       bSeparateFillAlpha=false
       outlineMaterial=null
       outlineSize=1
       outlineColor=<rgba 0 0 0 0.3>
      }
     }
     highlightColor {colorUseRule=UseColor_Specified, specifiedColor=<rgba 0 0 0 1>}
     highlightShape {
      bIsDynamicallyLoaded=false
      drawAs=Image
      imageType=NoImage
      mirroring=NoMirror
      resourceName=None
      resourceObject=null
      tiling=NoTile
      imageSize=<32 32>
      margin=<ltrb 0 0 0 0>
      outlineSettings {
       bUseBrushTransparency=false
       roundingType=HalfHeightRadius
       width=0
       color {colorUseRule=UseColor_Specified, specifiedColor=<rgba 0 0 0 0>}
       cornerRadii=<0 0 0 0>
      }
      tintColor {colorUseRule=UseColor_Specified, specifiedColor=<rgba 1 1 1 1>}
      uVRegion {bIsValid=false, max=<0 0>, min=<0 0>}
     }
     selectedBackgroundColor {colorUseRule=UseColor_Foreground, specifiedColor=<rgba 1 0 1 1>}
     shadowColorAndOpacity=<rgba 0 0 0 1>
     shadowOffset=<0 0>
     strikeBrush {
      bIsDynamicallyLoaded=false
      drawAs=Image
      imageType=NoImage
      mirroring=NoMirror
      resourceName=None
      resourceObject=null
      tiling=NoTile
      imageSize=<32 32>
      margin=<ltrb 0 0 0 0>
      outlineSettings {
       bUseBrushTransparency=false
       roundingType=HalfHeightRadius
       width=0
       color {colorUseRule=UseColor_Specified, specifiedColor=<rgba 0 0 0 0>}
       cornerRadii=<0 0 0 0>
      }
      tintColor {colorUseRule=UseColor_Specified, specifiedColor=<rgba 1 1 1 1>}
      uVRegion {bIsValid=false, max=<0 0>, min=<0 0>}
     }
     underlineBrush {
      bIsDynamicallyLoaded=false
      drawAs=Image
      imageType=NoImage
      mirroring=NoMirror
      resourceName=None
      resourceObject=null
      tiling=NoTile
      imageSize=<32 32>
      margin=<ltrb 0 0 0 0>
      outlineSettings {
       bUseBrushTransparency=false
       roundingType=HalfHeightRadius
       width=0
       color {colorUseRule=UseColor_Specified, specifiedColor=<rgba 0 0 0 0>}
       cornerRadii=<0 0 0 0>
      }
      tintColor {colorUseRule=UseColor_Specified, specifiedColor=<rgba 1 1 1 1>}
      uVRegion {bIsValid=false, max=<0 0>, min=<0 0>}
     }
    }
    secondNameColor=<rgba 0.7098 0.7569 0.7569 1>
   }
   slot {class=VerticalBoxSlot}
  }
  HorizontalBox_3 {widgetClassPath=&HorizontalBox, parent=VerticalBox_2, widget=HorizontalBox_3, slot {class=VerticalBoxSlot}}
  Image_HeroIcon {
   widgetClassPath=&Image
   bInherited=true
   bIsVariable=true
   parent=HorizontalBox_3
   widget=Image_HeroIcon
   properties {
    visibility=Collapsed
    brush {
     bIsDynamicallyLoaded=false
     drawAs=Image
     imageType=NoImage
     mirroring=NoMirror
     resourceName=None
     resourceObject=/Game/UI/Foundation/Textures/Icons/ItemTypes/T-Icon-Hero-Outlander-32
     tiling=NoTile
     imageSize=<24 24>
     margin=<ltrb 0 0 0 0>
     outlineSettings {
      bUseBrushTransparency=false
      roundingType=HalfHeightRadius
      width=0
      color {colorUseRule=UseColor_Specified, specifiedColor=<rgba 0 0 0 0>}
      cornerRadii=<0 0 0 0>
     }
     tintColor {colorUseRule=UseColor_Specified, specifiedColor=<rgba 1 1 1 1>}
     uVRegion {bIsValid=false, max=<0 0>, min=<0 0>}
    }
   }
   slot {class=HorizontalBoxSlot, properties {horizontalAlignment=HAlign_Center, verticalAlignment=VAlign_Center}}
  }
  PowerRatingWidgetSwitcher {
   widgetClassPath=&CommonWidgetSwitcherLegacy
   bInherited=true
   bIsVariable=true
   parent=HorizontalBox_3
   widget=PowerRatingWidgetSwitcher
   slot {class=HorizontalBoxSlot}
  }
  PowerRating_CampaignPower {
   widgetClassPath=/Game/UI/Foundation/Stats/PowerRatingWidget
   bInherited=true
   bIsVariable=true
   parent=PowerRatingWidgetSwitcher
   widget=PowerRating_CampaignPower
   properties {powerRatingType=Campaign, showDualPower=false, showProgress=false, useLargeStyle=false}
   slot {class=WidgetSwitcherSlot}
  }
  Moonbeam_PowerRating_CampaignPower {
   widgetClassPath=/Game/UI/Foundation/Stats/Moonbeam_PowerRatingWidget
   bInherited=true
   bIsVariable=true
   parent=PowerRatingWidgetSwitcher
   widget=Moonbeam_PowerRating_CampaignPower
   properties {powerRatingType=Campaign, showDualPower=false, showProgress=false, useLargeStyle=false}
   slot {class=WidgetSwitcherSlot, properties {padding=<ltrb 8 0 0 0>}}
  }
  Phoenix_PowerRating_CampaignPower {
   widgetClassPath=&Moonbeam_PowerRatingWidget
   bInherited=true
   bIsVariable=true
   parent=PowerRatingWidgetSwitcher
   widget=Phoenix_PowerRating_CampaignPower
   properties {powerRatingType=Phoenix, showDualPower=false, showProgress=false, useLargeStyle=false}
   slot {class=WidgetSwitcherSlot, properties {padding=<ltrb 8 0 0 0>}}
  }
  Text_ReadyStatus {
   widgetClassPath=&CommonTextBlock
   bInherited=true
   bIsVariable=true
   parent=VerticalBox_2
   widget=Text_ReadyStatus
   properties {
    justification=Center
    text=Not Ready
    colorAndOpacity {colorUseRule=UseColor_Specified, specifiedColor=<rgba 1 0.1714 0.1329 1>}
    font {
     bForceMonospaced=false
     bMaterialIsStencil=false
     fontMaterial=null
     fontObject=&BurbankSmall
     letterSpacing=0
     monospacedWidth=1
     size=16
     skewAmount=0
     typefaceFontName=Bold
     outlineSettings {
      bApplyOutlineToDropShadows=false
      bMiteredCorners=false
      bSeparateFillAlpha=false
      outlineMaterial=null
      outlineSize=1
      outlineColor=<rgba 0 0 0 0.33>
     }
    }
    shadowOffset=<0 0>
   }
   slot {class=VerticalBoxSlot, properties {horizontalAlignment=HAlign_Left}}
  }
  VoiceIcon_Status {
   widgetClassPath=/Script/FortniteUI.FortVoiceChatStatusIconProxyAnim
   bInherited=true
   bIsVariable=true
   parent=Overlay_3
   widget=VoiceIcon_Status
   slot {class=OverlaySlot, properties {horizontalAlignment=HAlign_Right, verticalAlignment=VAlign_Bottom, padding=<ltrb -32 0 0 0>}}
  }
  TeamMemberPedestal_ActivatableContainer {
   widgetClassPath=&/TeamMemberPedestal_ActivatableContainer
   bIsVariable=true
   parent=Overlay_3
   widget=TeamMemberPedestal_ActivatableContainer
   slot {class=OverlaySlot, properties {padding=<ltrb 6 -40 0 0>}}
  }
  Button_AddFriend {
   widgetClassPath=&/TeamMemberPedestal_Nameplate_AddFriendButton
   bInherited=true
   bIsVariable=true
   namedSlotHost=TeamMemberPedestal_ActivatableContainer
   widget=Button_AddFriend
   properties {
    onButtonBaseClicked=(&self:WidgetTree.Button_AddFriend.HandleOnButtonClicked)
    onButtonBaseDoubleClicked=(&self:WidgetTree.Button_AddFriend.HandleOnButtonDoubleClicked)
    onButtonBaseHovered=(&self:WidgetTree.Button_AddFriend.HandleOnButtonHovered)
    onButtonBaseUnhovered=(&self:WidgetTree.Button_AddFriend.HandleOnButtonUnhovered)
    onSelectedChangedBase=(&self:WidgetTree.Button_AddFriend.HandleOnSelectedChanged)
   }
  }
 ]
}
