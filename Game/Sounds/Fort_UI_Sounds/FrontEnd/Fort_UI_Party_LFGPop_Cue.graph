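# Front-end UI Sound Cue for a party LFG pop, routed to the Fort_UI sound class.
# A single party-add pop wave plays through a Modulator that randomizes pitch upward and volume downward slightly, under a reduced cue volume.
# Virtualization is disabled, and a concurrency override caps playback at two instances, stopping the oldest, with volume scaling that quickly ducks overlapping plays.
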SoundCue /Game/Sounds/Fort_UI_Sounds/FrontEnd/Fort_UI_Party_LFGPop_Cue
root=Modulator_0
nodes [
 {id=Modulator_0, children [WavePlayer_0], properties {pitchMax=1.2, pitchMin=1, volumeMax=1, volumeMin=0.8}}
 {id=WavePlayer_0, properties {soundWaveAssetPtr=&/Fort_UI_Party_Add_Pop_01}}
]
properties {
 bOverrideConcurrency=true
 maxDistance=2097152
 soundClassObject=&Fort_UI
 virtualizationMode=Disabled
 volumeMultiplier=0.85
 concurrencyOverrides {bVolumeScaleCanRelease=true, maxCount=2, resolutionRule=StopOldest, volumeScale=0, volumeScaleAttackTime=0.2}
}
