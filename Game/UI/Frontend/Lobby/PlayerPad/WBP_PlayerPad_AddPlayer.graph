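# Lobby player pad widget that draws a circular "add player" slot with a centered plus-style Add icon.
# A GridPanel layers a fixed-size spacer, an obscured circular UIKit backplate, the icon, and a UIKit Block Outline on top.
# OnTransitionFocus and OnTransitionUnfocus events request Focused or Unfocused visual states on the outline.
# The widget holds no variables or click handling, and nothing in its graph calls these events.
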WidgetBlueprint /Game/UI/Frontend/Lobby/PlayerPad/WBP_PlayerPad_AddPlayer
graphs [EventGraph {nodes [
   OnTransitionUnfocus() {is=title K2Node_CustomEvent}
   n1 = Get Block_Outline() -> (Block_Outline:WBP UIKit Block Outline Object Reference)
   > RequestTransition!(execute, self:UIKit Block Object Reference=@n1, State:EUIKitBlockVisualState Enum=Unfocused) -> (then)
   OnTransitionFocus() {is=title K2Node_CustomEvent}
   > RequestTransition(self=@n1, State=Focused)
  ]}]
variables []
widgets {
 info {inheritedWidgetCount=0, parentClass=&UserWidget, rootWidgetClass=&GridPanel, widgetCount=5}
 widgets [
  GridPanel_0 {widgetClassPath=&GridPanel, widget=GridPanel_0, properties {columnFill [1], rowFill [1]}}
  Spacer_209 {
   widgetClassPath=&Spacer
   parent=GridPanel_0
   widget=Spacer_209
   properties {size=<56 56>}
   slot {class=GridSlot, properties {layer=-50}}
  }
  WBP_UIKit_Backplate {
   widgetClassPath=&WBP_UIKit_Backplate
   bIsVariable=true
   parent=GridPanel_0
   widget=WBP_UIKit_Backplate
   properties {brightness=Obscure, cornerRadius=Circle}
   slot {class=GridSlot, properties {layer=-10}}
  }
  AddIcon {
   widgetClassPath=&Image
   parent=GridPanel_0
   widget=AddIcon
   properties {
    visibility=SelfHitTestInvisible
    brush {
     bIsDynamicallyLoaded=false
     drawAs=Image
     imageType=FullColor
     mirroring=NoMirror
     resourceName=None
     resourceObject=/FortUILibrary/Textures/MaterialInstances/MI_UIKit_Icon_Add
     tiling=NoTile
     imageSize=<32 32>
     margin=<ltrb 0 0 0 0>
     outlineSettings {
      bUseBrushTransparency=false
      roundingType=HalfHeightRadius
      width=0
      color {colorUseRule=UseColor_Specified, specifiedColor=<rgba 0 0 0 0>}
      cornerRadii=<0 0 0 0>
     }
     tintColor {colorUseRule=UseColor_Specified, specifiedColor=<rgba 1 1 1 1>}
     uVRegion {bIsValid=false, max=<0 0>, min=<0 0>}
    }
   }
   slot {class=GridSlot, properties {horizontalAlignment=HAlign_Center, verticalAlignment=VAlign_Center}}
  }
  Block_Outline {
   widgetClassPath=/FortUILibrary/UIKit/Buttons/Blocks/WBP_UIKit_Block_Outline
   bIsVariable=true
   parent=GridPanel_0
   widget=Block_Outline
   properties {bottomLeftCornerRadius=0.5, bottomRightCornerRadius=0.5, topLeftCornerRadius=0.5, topRightCornerRadius=0.5}
   slot {class=GridSlot, properties {layer=1}}
  }
 ]
}
