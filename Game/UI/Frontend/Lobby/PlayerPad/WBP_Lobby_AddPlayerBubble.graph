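# Lobby add-player bubble widget (FortLobbyAddPlayer child) for an empty party slot, stacking a party suggestion button, an empty-slot button and an input-action prompt.
# The empty-slot button hosts the add-player pad visual, which transitions in and out of focus on hover and unhover.
# Construct caches the cast suggestion button, whose focus transition also follows hover, but only on gamepad.
# Hover and unhover each play their own 2D sound for any input type.
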WidgetBlueprint /Game/UI/Frontend/Lobby/PlayerPad/WBP_Lobby_AddPlayerBubble
graphs [EventGraph {nodes [
   Event On Hovered()
   n1 = Get WBP_PlayerPad_AddPlayer() -> (WBP_PlayerPad_AddPlayer:WBP Player Pad Add Player Object Reference)
   > OnTransitionFocus(self:WBP Player Pad Add Player Object Reference=@n1)
   n2 = "Equal (Enum)"(A=GetCurrentInputType(self=Get CommonInputSubsystem()), B=Gamepad)
   > n3 = Branch(Condition=@n2)
   n4 = Get As WBP Party Suggestion Button() -> (As WBP Party Suggestion Button:WBP Party Suggestion Button Object Reference)
   > n5 = SetTransitionFocus!(execute, self:WBP Party Suggestion Button Object Reference=@n4, IsFocused:bool=true) -> (then)
   SpawnSound2D!(
    execute=[@n5.then, @n3.else]
    Sound:Sound Base Object Reference=/Game/Sounds/Fort_UI_Sounds/FrontEnd/Fort_UI_Party_LFGPop_Cue
    VolumeMultiplier:float=1
    PitchMultiplier:float=1
    StartTime:float=0
    ConcurrencySettings:Sound Concurrency Object Reference
    bPersistAcrossLevelTransition:bool=false
    bAutoDestroy:bool=true
   ) -> (then, ReturnValue:Audio Component Object Reference)
   Event On Unhovered()
   > OnTransitionUnfocus(self:WBP Player Pad Add Player Object Reference=@n1)
   > n6 = Branch(Condition=@n2)
   > n7 = SetTransitionFocus(self=@n4, IsFocused=false)
   SpawnSound2D(
    execute=[@n7.then, @n6.else]
    Sound=/Game/Sounds/Fort_UI_Sounds/FrontEnd/Fort_UI_Party_UnHover_Cue
    VolumeMultiplier=1
    PitchMultiplier=1
    StartTime=0
    bPersistAcrossLevelTransition=false
    bAutoDestroy=true
   )
   Event Construct()
   > n8 = Cast To WBP_PartySuggestionButton!(execute, Object=Get PartySuggestion() -> (PartySuggestion:Fort Party Suggestion Button Object Reference)) -> (then, CastFailed, AsWBP Party Suggestion Button:WBP Party Suggestion Button Object Reference) {is=title K2Node_DynamicCast}
   > Set As WBP Party Suggestion Button(As WBP Party Suggestion Button:WBP Party Suggestion Button Object Reference=@n8) -> (Output_Get:WBP Party Suggestion Button Object Reference)
  ]}]
variables [As WBP Party Suggestion Button {type=WBP_PartySuggestionButton_C, default=null}]
widgets {
 info {
  inheritedWidgetCount=3
  namedSlotCount=2
  parentClass=/Script/FortniteUI.FortLobbyAddPlayer
  rootWidgetClass=/Game/Athena/HUD/FortHUDElementWrapper.FortHUDElementWrapper_C
  widgetCount=8
 }
 widgets [
  FortHUDElementWrapper {
   widgetClassPath=/Game/Athena/HUD/FortHUDElementWrapper
   bIsVariable=true
   widget=FortHUDElementWrapper
   properties {elementTag {tagName=HUD}}
  }
  VerticalBox_0 {widgetClassPath=&VerticalBox, namedSlotHost=FortHUDElementWrapper, widget=VerticalBox_0}
  PartySuggestion {
   widgetClassPath=/Game/Athena/UI/Frontend/Parties/WBP_PartySuggestionButton
   bInherited=true
   bIsVariable=true
   parent=VerticalBox_0
   widget=PartySuggestion
   properties {
    onButtonBaseClicked=(&self:WidgetTree.PartySuggestion.HandleOnButtonClicked)
    onButtonBaseDoubleClicked=(&self:WidgetTree.PartySuggestion.HandleOnButtonDoubleClicked)
    onButtonBaseHovered=(&self:WidgetTree.PartySuggestion.HandleOnButtonHovered)
    onButtonBaseUnhovered=(&self:WidgetTree.PartySuggestion.HandleOnButtonUnhovered)
    onSelectedChangedBase=(&self:WidgetTree.PartySuggestion.HandleOnSelectedChanged)
    visibility=Collapsed
   }
   slot {class=VerticalBoxSlot, properties {padding=<ltrb 0 0 0 40>}}
  }
  OverlayEmptySlot {
   widgetClassPath=&Overlay
   parent=VerticalBox_0
   widget=OverlayEmptySlot
   slot {class=VerticalBoxSlot, properties {horizontalAlignment=HAlign_Center, verticalAlignment=VAlign_Center}}
  }
  Button_EmptySlot {
   widgetClassPath=/FortUILibrary/Foundation/Buttons/PanelButton
   bInherited=true
   bIsVariable=true
   parent=OverlayEmptySlot
   widget=Button_EmptySlot
   properties {
    bInteractableWhenSelected=true
    bSelectable=true
    bShouldUseFallbackDefaultInputAction=false
    clickMethod=MouseDown
    onButtonBaseClicked=(&self:WidgetTree.Button_EmptySlot.HandleOnButtonClicked)
    onButtonBaseDoubleClicked=(&self:WidgetTree.Button_EmptySlot.HandleOnButtonDoubleClicked)
    onButtonBaseHovered=(&self:WidgetTree.Button_EmptySlot.HandleOnButtonHovered)
    onButtonBaseUnhovered=(&self:WidgetTree.Button_EmptySlot.HandleOnButtonUnhovered)
    onSelectedChangedBase=(&self:WidgetTree.Button_EmptySlot.HandleOnSelectedChanged)
    style=/FortUILibrary/Foundation/Buttons/ButtonStyles/ButtonStyle-TextOnlyBase_Empty.ButtonStyle-TextOnlyBase_Empty_C
    hoveredSlateSoundOverride {resourceObject=/Game/Sounds/Fort_UI_Sounds/FrontEnd/DefaultSounds/PressedButtons/MSS_UI_Button_Hover_Default}
   }
   slot {class=OverlaySlot}
  }
  WBP_PlayerPad_AddPlayer {
   widgetClassPath=&/WBP_PlayerPad_AddPlayer
   bIsVariable=true
   namedSlotHost=Button_EmptySlot
   widget=WBP_PlayerPad_AddPlayer
  }
  OverlayInputAction {
   widgetClassPath=&Overlay
   bIsVariable=true
   parent=VerticalBox_0
   widget=OverlayInputAction
   slot {class=VerticalBoxSlot, properties {horizontalAlignment=HAlign_Center, verticalAlignment=VAlign_Top}}
  }
  InputActionWidget {
   widgetClassPath=&CommonActionWidget
   bInherited=true
   parent=OverlayInputAction
   widget=InputActionWidget
   properties {visibility=Hidden, inputActions [{dataTable=&InputActionDataTable, rowName=Select}]}
   slot {class=OverlaySlot, properties {padding=<ltrb 0 24 0 0>}}
  }
 ]
}
